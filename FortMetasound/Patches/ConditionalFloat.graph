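# ConditionalFloat is a MetaSound patch that outputs either its True or False float input according to a Bool Condition.
# It converts Condition to a float, multiplies True by that value and False by one minus it, then adds the two products.
# The selection is done purely with arithmetic nodes, with no branching or triggers.
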MetaSoundPatch /FortMetasound/Patches/ConditionalFloat
inputs [
 Condition {type=Bool}
 False {type=Float}
 True {type=Float}
]
interfaces [MetaSound:1.0]
outputs [Output {type=Float}]
pages [Default {
  input_defaults {Condition=null, False=0, True=0}
  nodes [
   n1 = Convert.Float:Bool(Bool=^Condition)
   Add_0 = UE.Add:Float(
    PrimaryOperand=UE.Multiply:Float(AdditionalOperands=^True, PrimaryOperand=@n1.Float).Out
    AdditionalOperands=UE.Multiply:Float(AdditionalOperands=^False, PrimaryOperand=UE.Subtract:Float(AdditionalOperands=@n1.Float) {also {defaults {PrimaryOperand=1}}}.Out).Out
   )
  ]
  links [{dst="@Output", src=Add_0.Out}]
 }]
