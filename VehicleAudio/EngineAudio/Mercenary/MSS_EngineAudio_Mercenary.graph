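# Looping mono MetaSound source for the Mercenary vehicle engine, layering a REV player with loops and one-shots.
# RPM, throttle and gear drive the REV core; speed and gear shape loop pitch, with RPM pitching a synth layer.
# Bonus speed crossfades loop variants and fires a one-shot, while leaving the ground fades in a delayed flying loop.
# Shuffled, gear-pitched shift sounds join the high-passed mix, where upshifts briefly duck loops and boost REV.
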MetaSoundSource /VehicleAudio/EngineAudio/Mercenary/MSS_EngineAudio_Mercenary
inputs [
 Array {type=WaveAsset:Array, access=Constructor}
 Array 1 {type=WaveAsset:Array, access=Constructor}
 Character.Affiliation {type=Enum:FortAudioAffiliation}
 Character.Health {type=Float}
 Character.UniqueID {type=Int32}
 Downshift {type=Trigger}
 DownshiftWaveArray {type=WaveAsset:Array, access=Constructor}
 Fort.Vehicle.Acceleration {type=Float}
 Fort.Vehicle.BlowoffStrength {type=Float}
 Fort.Vehicle.BonusSpeed {type=Float}
 Fort.Vehicle.EnableShifting {type=Bool}
 Fort.Vehicle.Gear {type=Int32}
 Fort.Vehicle.IsGrounded {type=Bool}
 Fort.Vehicle.Load {type=Float}
 Fort.Vehicle.NormalSpeed {type=Float}
 Fort.Vehicle.OnBlowoff {type=Trigger}
 Fort.Vehicle.OnDownshift {type=Trigger}
 Fort.Vehicle.OnUpshift {type=Trigger}
 Fort.Vehicle.Pitch {type=Float}
 Fort.Vehicle.RPM {type=Float}
 Fort.Vehicle.Throttle {type=Float}
 REV Asset {type=REV Asset}
 REV.EnableShifting {type=Bool}
 REV.Gear {type=Int32}
 REV.RPM {type=Float}
 REV.Throttle {type=Float}
 REV.Velocity {type=Float}
 Speed_Normalized {type=Float}
 UE.Attenuation.Distance {type=Float}
 UE.Source.OnPlay {type=Trigger}
 UE.Source.Orientation.Azimuth {type=Float}
 UE.Source.Orientation.Elevation {type=Float}
 Upshift {type=Trigger}
 UpshiftWaveArray {type=WaveAsset:Array, access=Constructor}
]
interfaces [Fort.Character:1.0, Fort.Vehicle:1.0, REV:1.0, UE.Attenuation:1.0, UE.OutputFormat.Mono:1.0, UE.Source.Orientation:1.0, UE.Source:1.1]
outputs ["UE.OutputFormat.Mono.Audio:0" {type=Audio}]
pages [Default {
  input_defaults {
   Character.Affiliation=2
   Character.Health=0
   Character.UniqueID=0
   Downshift=null
   Fort.Vehicle.Acceleration=0
   Fort.Vehicle.BlowoffStrength=0
   Fort.Vehicle.BonusSpeed=0
   Fort.Vehicle.EnableShifting=null
   Fort.Vehicle.Gear=0
   Fort.Vehicle.IsGrounded="<MetaSoundBuilderResult.SUCCEEDED: 0>"
   Fort.Vehicle.Load=0
   Fort.Vehicle.NormalSpeed=0
   Fort.Vehicle.OnBlowoff=null
   Fort.Vehicle.OnDownshift=null
   Fort.Vehicle.OnUpshift=null
   Fort.Vehicle.Pitch=0
   Fort.Vehicle.RPM=0
   Fort.Vehicle.Throttle=0
   REV Asset=&/Engine_Mecenary_Mono_03
   REV.EnableShifting=null
   REV.Gear=0
   REV.RPM=0
   REV.Throttle=0
   REV.Velocity=0
   Speed_Normalized=0
   UE.Attenuation.Distance=0
   UE.Source.OnPlay=null
   UE.Source.Orientation.Azimuth=0
   UE.Source.Orientation.Elevation=0
   Upshift=null
   Array [
    &/EngineAudio_Mecenary_Flying_Loop_01
    &/EngineAudio_Mecenary_Driving_Flying_Loop_02
   ]
   Array 1 [
    &/EngineAudio_Mecenary_Driving_Loop_01
    &EngineAudio_Mecenary_Flying_Loop_01
   ]
   DownshiftWaveArray [
    &/EngineAudio_Mercenary_DownShift_04
    &/EngineAudio_Mercenary_DownShift_03
    &/EngineAudio_Mercenary_DownShift_02
    &/EngineAudio_Mercenary_DownShift_01
   ]
   UpshiftWaveArray [
    &/EngineAudio_Mercenary_UpShift_01
    &/EngineAudio_Mercenary_UpShift_02
    &/EngineAudio_Mercenary_UpShift_03
    &/EngineAudio_Mercenary_UpShift_04
   ]
  }
  nodes [
   TriggerOnce_0 = UE.Trigger Once(
    Reset=UE.TriggerOnThreshold:Float(In=^Fort.Vehicle.BonusSpeed) {also {defaults {Threshold=0.1, Type=1}}}.Out
    Trigger In=UE.TriggerOnThreshold:Float(In=^Fort.Vehicle.BonusSpeed) {also {defaults {Threshold=0.9}}}.Out
   )
   WavePlayer_7 = UE.Wave Player:Mono(Play=@TriggerOnce_0.Trigger Out) {also {defaults {Wave Asset=&/EngineAudio_Mecenary_BonusSpeed_Gained_OneShot_01}, version [1,1]}}
   TriggerAnyTriggerAny2_3 = "TriggerAny.Trigger Any (2)"(In 0=^Fort.Vehicle.OnUpshift, In 1=^Upshift)
   n1 = VariableAccessor.Trigger(Variable=VariableMutator.Trigger(Variable=InitVariable.Trigger().Variable, Value=@TriggerAnyTriggerAny2_3.Out).Variable)
   n2 = VariableAccessor.Trigger(Variable=@n1.Variable)
   n3 = Array.Shuffle:WaveAsset:Array(In Array=^UpshiftWaveArray, Next=@n2.Value)
   n4 = UE.Reroute(Value=Max.Max:Int32(B=^Fort.Vehicle.Gear, A=^REV.Gear).Value)
   n5 = VariableAccessor.Int32(Variable=VariableMutator.Int32(Variable=InitVariable.Int32().Variable, Value=@n4.Value).Variable)
   n6 = VariableAccessor.Int32(Variable=@n5.Variable)
   TriggerAnyTriggerAny2_5 = "TriggerAny.Trigger Any (2)"(In 0=^Downshift, In 1=^Fort.Vehicle.OnDownshift)
   n7 = VariableAccessor.Trigger(Variable=VariableMutator.Trigger(Variable=InitVariable.Trigger().Variable, Value=@TriggerAnyTriggerAny2_5.Out).Variable)
   WavePlayer_0 = UE.Wave Player:Mono(
    Play=@n3.On Next
    Wave Asset=@n3.Value
    Pitch Shift=MapRange.MapRange:Float(In=Convert.Float:Int32(Int32=@n6.Value).Float) {also {defaults {In Range B=4, Out Range B=2}}}.Out Value
    Stop=VariableAccessor.Trigger(Variable=@n7.Variable).Value
   ) {also {version [1,1]}}
   n8 = Array.Shuffle:WaveAsset:Array(In Array=^DownshiftWaveArray, Next=@n7.Value)
   n9 = VariableAccessor.Int32(Variable=@n6.Variable)
   WavePlayer_1 = UE.Wave Player:Mono(
    Play=@n8.On Next
    Wave Asset=@n8.Value
    Pitch Shift=MapRange.MapRange:Float(In=Convert.Float:Int32(Int32=@n9.Value).Float) {also {defaults {In Range B=4, Out Range A=-2, Out Range B=0}}}.Out Value
    Stop=VariableAccessor.Trigger(Variable=@n2.Variable).Value
   ) {also {version [1,1]}}
   AudioMixerAudioMixerMono3_0 = "AudioMixer.Audio Mixer (Mono, 3)"(
    In 0=UE.Reroute(Value=@WavePlayer_7.Out Mono).Value
    In 1=UE.Reroute(Value=@WavePlayer_0.Out Mono).Value
    In 2=UE.Reroute(Value=@WavePlayer_1.Out Mono).Value
   ) {also {defaults {Gain 0=1.4962}}}
   n10 = VariableAccessor.Float(Variable=VariableMutator.Float(Variable=InitVariable.Float().Variable, Value=Max.Max:Float(B=^Fort.Vehicle.NormalSpeed, A=^Speed_Normalized).Value).Variable)
   n11 = VariableAccessor.Float(Variable=VariableMutator.Float(
     Variable=InitVariable.Float().Variable
     Value=/FortMetasound/Patches/MapRange_DecibelOutput(In=@n10.Value) {also {defaults {Out Range A=-6}}}.Out Value
    ).Variable)
   TriggerOnValueChange_3 = UE.Trigger On Value Change:Bool(Value=^Fort.Vehicle.IsGrounded)
   TriggerAnyTriggerAny3_0 = "TriggerAny.Trigger Any (3)"(In 0=^UE.Source.OnPlay, In 1=@TriggerOnValueChange_3.Trigger)
   n12 = TriggerCompare.Clamp:Bool(A=^Fort.Vehicle.IsGrounded, Compare=@TriggerAnyTriggerAny3_0.Out) {also {defaults {B="<MetaSoundBuilderResult.SUCCEEDED: 0>"}}}
   TriggerDelay_0 = UE.Trigger Delay(In=@n12.False, Reset=@n12.True) {also {defaults {Delay Time=0.25}}}
   TriggerRouteTriggerRouteBool2_0 = "TriggerRoute.Trigger Route (Bool, 2):Bool"(Set 0=@n12.True, Set 1=@TriggerDelay_0.Out) {also {defaults {Value 0="<MetaSoundBuilderResult.SUCCEEDED: 0>"}}}
   n13 = VariableAccessor.Bool(Variable=VariableMutator.Bool(Variable=InitVariable.Bool().Variable, Value=@TriggerRouteTriggerRouteBool2_0.Value).Variable)
   TriggerOnValueChange_2 = UE.Trigger On Value Change:Bool(Value=@n13.Value)
   TriggerAnyTriggerAny2_0 = "TriggerAny.Trigger Any (2)"(In 0=^UE.Source.OnPlay, In 1=@TriggerOnValueChange_2.Trigger)
   n14 = TriggerCompare.Clamp:Bool(Compare=@TriggerAnyTriggerAny2_0.Out, A=@n13.Value)
   ADSREnvelopeADSREnvelope_4 = ADSR Envelope.ADSR Envelope:Float(Trigger Release=@n14.False, Trigger Attack=@n14.True) {also {defaults {Attack Time=0.5, Hard Reset="<MetaSoundBuilderResult.SUCCEEDED: 0>", Sustain Level=1}}}
   n15 = UE.Multiply:Float(AdditionalOperands=@ADSREnvelopeADSREnvelope_4.Out Envelope) {also {defaults {PrimaryOperand=1.9953}}}
   WavePlayer_2 = UE.Wave Player:Mono(Play=@ADSREnvelopeADSREnvelope_4.On Attack Triggered, Stop=@ADSREnvelopeADSREnvelope_4.On Done) {also {
     defaults {Loop="<MetaSoundBuilderResult.SUCCEEDED: 0>", Wave Asset=&EngineAudio_Mecenary_Driving_Flying_Loop_02}
     version [1,1]
    }}
   TriggerOnValueChange_4 = UE.Trigger On Value Change:Float(Value=^Fort.Vehicle.BonusSpeed)
   TriggerAnyTriggerAny2_7 = "TriggerAny.Trigger Any (2)"(In 1=^UE.Source.OnPlay, In 0=@TriggerOnValueChange_4.Trigger)
   TriggerAnyTriggerAny2_1 = "TriggerAny.Trigger Any (2)"(In 0=^UE.Source.OnPlay, In 1=@TriggerOnValueChange_4.Trigger)
   TriggerRouteTriggerRouteBool2_1 = "TriggerRoute.Trigger Route (Bool, 2):Bool"(
    Set 1=TriggerCompare.Clamp:Float(A=^Fort.Vehicle.BonusSpeed, Compare=@TriggerAnyTriggerAny2_7.Out) {also {defaults {B=0.1, Type=2}}}.True
    Set 0=TriggerCompare.Clamp:Float(A=^Fort.Vehicle.BonusSpeed, Compare=@TriggerAnyTriggerAny2_1.Out) {also {defaults {B=0.1, Type=3}}}.True
   ) {also {defaults {Value 0="<MetaSoundBuilderResult.SUCCEEDED: 0>"}}}
   n16 = VariableDeferredAccessor.Bool(Variable=InitVariable.Bool().Variable)
   n17 = VariableDeferredAccessor.Bool(Variable=@n16.Variable)
   n18 = VariableAccessor.Bool(Variable=VariableMutator.Bool(Value=@TriggerRouteTriggerRouteBool2_1.Value, Variable=@n17.Variable).Variable)
   n19 = VariableAccessor.Bool(Variable=@n18.Variable)
   n20 = VariableAccessor.Bool(Variable=@n19.Variable)
   n21 = VariableAccessor.Bool(Variable=@n20.Variable)
   n22 = Convert.Float:Bool(Bool=@n21.Value)
   n23 = UE.InterpTo:Audio(Target=MapRange.MapRange:Float(In=UE.Subtract:Float(AdditionalOperands=@n22.Float) {also {defaults {PrimaryOperand=1}}}.Out) {also {defaults {Out Range A=0.4}}}.Out Value) {also {defaults {Interp Time=0.5}}}
   TriggerOnValueChange_0 = UE.Trigger On Value Change:Bool(Value=@n19.Value)
   TriggerAnyTriggerAny2_4 = "TriggerAny.Trigger Any (2)"(In 0=^UE.Source.OnPlay, In 1=@TriggerOnValueChange_0.Trigger)
   n24 = TriggerCompare.Clamp:Bool(Compare=@TriggerAnyTriggerAny2_4.Out, A=@n19.Value) {also {defaults {B="<MetaSoundBuilderResult.SUCCEEDED: 0>"}}}
   ADSREnvelopeADSREnvelope_0 = ADSR Envelope.ADSR Envelope:Float(Trigger Attack=@n24.False, Trigger Release=@n24.True) {also {defaults {Attack Time=0.1, Hard Reset="<MetaSoundBuilderResult.SUCCEEDED: 0>", Release Time=0.1, Sustain Level=1}}}
   ADSREnvelopeADSREnvelope_6 = ADSR Envelope.ADSR Envelope:Float(Trigger Release=@n24.False, Trigger Attack=@n24.True) {also {defaults {Attack Time=0.1, Hard Reset="<MetaSoundBuilderResult.SUCCEEDED: 0>", Release Time=0.1, Sustain Level=1}}}
   n25 = MapRange.MapRange:Float(In=Convert.Float:Int32(Int32=@n5.Value).Float) {also {defaults {In Range B=5, Out Range B=4.5}}}
   n26 = VariableAccessor.Float(Variable=@n10.Variable)
   n27 = MapRange.MapRange:Float(In=@n26.Value) {also {defaults {Out Range A=-4, Out Range B=6}}}
   ADSREnvelopeADSREnvelope_3 = ADSR Envelope.ADSR Envelope:Float(Trigger Attack=^UE.Source.OnPlay) {also {defaults {Attack Time=1, Sustain Level=1}}}
   n28 = VariableAccessor.Float(Variable=VariableMutator.Float(
     Variable=InitVariable.Float().Variable
     Value=UE.Subtract:Float(AdditionalOperands=@ADSREnvelopeADSREnvelope_3.Out Envelope) {also {defaults {PrimaryOperand=1}}}.Out
    ).Variable)
   n29 = VariableAccessor.Float(Variable=@n28.Variable)
   n30 = UE.Add:Float(
    PrimaryOperand=&InterpSpeed(FloatInput=UE.Add:Float(AdditionalOperands=@n25.Out Value, PrimaryOperand=@n27.Out Value).Out) {also {defaults {FallingInterpSpeed=0.1, RisingInterpSpeed=0.75}}}.FloatOutput
    AdditionalOperands=UE.Multiply:Float(PrimaryOperand=@n29.Value) {also {defaults {AdditionalOperands=-6}}}.Out
   )
   n31 = MapRange.MapRange:Float(In=VariableAccessor.Float(Variable=@n26.Variable).Value) {also {defaults {Out Range A=0.15, Out Range B=0}}}
   n32 = UE.LFO:Audio(Max Value=@n31.Out Value, Min Value=UE.Multiply:Float(PrimaryOperand=@n31.Out Value) {also {defaults {AdditionalOperands=-1}}}.Out) {also {defaults {Frequency=2}}}
   n33 = UE.Reroute(Value=UE.Reroute(Value=UE.Add:Float(PrimaryOperand=@n30.Out, AdditionalOperands=@n32.Out).Out).Value)
   WavePlayer_4 = UE.Wave Player:Mono(Play=@ADSREnvelopeADSREnvelope_0.On Attack Triggered, Stop=@ADSREnvelopeADSREnvelope_0.On Done, Pitch Shift=@n33.Value) {also {
     defaults {Loop="<MetaSoundBuilderResult.SUCCEEDED: 0>", Wave Asset=&EngineAudio_Mecenary_Driving_Loop_01}
     version [1,1]
    }}
   WavePlayer_6 = UE.Wave Player:Mono(Play=@ADSREnvelopeADSREnvelope_6.On Attack Triggered, Stop=@ADSREnvelopeADSREnvelope_6.On Done, Pitch Shift=@n33.Value) {also {
     defaults {Loop="<MetaSoundBuilderResult.SUCCEEDED: 0>", Wave Asset=&EngineAudio_Mecenary_Flying_Loop_01}
     version [1,1]
    }}
   AudioMixerAudioMixerMono2_0 = "AudioMixer.Audio Mixer (Mono, 2)"(
    Gain 1=@ADSREnvelopeADSREnvelope_0.Out Envelope
    Gain 0=@ADSREnvelopeADSREnvelope_6.Out Envelope
    In 1=@WavePlayer_4.Out Mono
    In 0=@WavePlayer_6.Out Mono
   )
   n34 = UE.Reroute(Value=@AudioMixerAudioMixerMono2_0.Out)
   TriggerOnValueChange_1 = UE.Trigger On Value Change:Bool(Value=@n20.Value)
   TriggerAnyTriggerAny2_6 = "TriggerAny.Trigger Any (2)"(In 0=^UE.Source.OnPlay, In 1=@TriggerOnValueChange_1.Trigger)
   TriggerAnyTriggerAny2_2 = "TriggerAny.Trigger Any (2)"(
    In 0=^UE.Source.OnPlay
    In 1=TriggerCompare.Clamp:Bool(Compare=@TriggerAnyTriggerAny2_6.Out, A=@n20.Value) {also {defaults {B="<MetaSoundBuilderResult.SUCCEEDED: 0>"}}}.False
   )
   n35 = TriggerCompare.Clamp:Bool(Compare=@TriggerAnyTriggerAny2_2.Out, A=VariableAccessor.Bool(Variable=@n21.Variable).Value)
   ADSREnvelopeADSREnvelope_1 = ADSR Envelope.ADSR Envelope:Float(Trigger Release=@n35.False, Trigger Attack=@n35.True) {also {defaults {Attack Time=0.1, Hard Reset="<MetaSoundBuilderResult.SUCCEEDED: 0>", Release Time=0.1, Sustain Level=1}}}
   n36 = UE.Multiply:Float(PrimaryOperand=UE.Reroute(Value=@ADSREnvelopeADSREnvelope_1.Out Envelope).Value) {also {defaults {AdditionalOperands=0.43}}}
   n37 = UE.Reroute(Value=Max.Max:Float(B=^Fort.Vehicle.RPM, A=^REV.RPM).Value)
   n38 = VariableAccessor.Float(Variable=VariableMutator.Float(Variable=InitVariable.Float().Variable, Value=@n37.Value).Variable)
   n39 = MapRange.MapRange:Float(In=UE.Multiply:Float(AdditionalOperands=@n38.Value, PrimaryOperand=@n38.Value).Out) {also {defaults {Out Range A=-1, Out Range B=6.03}}}
   n40 = UE.Multiply:Float(PrimaryOperand=@n28.Value) {also {defaults {AdditionalOperands=-12}}}
   n41 = MapRange.MapRange:Float(In=Convert.Float:Bool(Bool=@n18.Value).Float) {also {defaults {Out Range B=0.75}}}
   WavePlayer_3 = UE.Wave Player:Mono(
    Play=@ADSREnvelopeADSREnvelope_1.On Attack Triggered
    Stop=@ADSREnvelopeADSREnvelope_1.On Done
    Pitch Shift=UE.Add:Float(
     PrimaryOperand=UE.Add:Float(AdditionalOperands=@n39.Out Value, PrimaryOperand=@n40.Out).Out
     AdditionalOperands=UE.InterpTo:Audio(Target=@n41.Out Value) {also {defaults {Interp Time=0.3}}}.Value
    ).Out
   ) {also {
     defaults {Loop="<MetaSoundBuilderResult.SUCCEEDED: 0>", Wave Asset=&/EngineAudio_Mercenary_Synth_Driving_Loop}
     version [1,1]
    }}
   ADSREnvelopeADSREnvelope_5 = ADSR Envelope.ADSR Envelope:Float(Trigger Attack=@n35.False, Trigger Release=@n35.True) {also {defaults {Attack Time=0.1, Hard Reset="<MetaSoundBuilderResult.SUCCEEDED: 0>", Release Time=0.1, Sustain Level=1}}}
   WavePlayer_5 = UE.Wave Player:Mono(Play=@ADSREnvelopeADSREnvelope_5.On Attack Triggered, Stop=@ADSREnvelopeADSREnvelope_5.On Done) {also {
     defaults {Loop="<MetaSoundBuilderResult.SUCCEEDED: 0>", Wave Asset=&/EngineAudio_Mercenary_Synth_Driving_Fast_Loop}
     version [1,1]
    }}
   AudioMixerAudioMixerMono2_1 = "AudioMixer.Audio Mixer (Mono, 2)"(Gain 0=@n36.Out, Gain 1=@n36.Out, In 0=@WavePlayer_3.Out Mono, In 1=@WavePlayer_5.Out Mono)
   AudioMixerAudioMixerMono3_1 = "AudioMixer.Audio Mixer (Mono, 3)"(
    Gain 1=UE.Multiply:Float(
     AdditionalOperands=/FortMetasound/Patches/MSP_DistanceMapper(Distance=^UE.Attenuation.Distance) {also {defaults {Falloff=1500, Final Volume=0.5088, Inner Radius=1000, Inner Volume=1}}}.Output
     PrimaryOperand=UE.Multiply:Float(PrimaryOperand=@n11.Value) {also {defaults {AdditionalOperands=0.8913}}}.Out
    ).Out
    Gain 2=UE.Multiply:Float(PrimaryOperand=VariableAccessor.Float(Variable=@n11.Variable).Value) {also {defaults {AdditionalOperands=1.4125}}}.Out
    In 0=UE.Reroute(Value=UE.Multiply:Audio by Float(AdditionalOperands=@n15.Out, PrimaryOperand=@WavePlayer_2.Out Mono).Out).Value
    In 2=UE.Reroute(Value=UE.Multiply:Audio by Float(AdditionalOperands=@n23.Value, PrimaryOperand=@n34.Value).Out).Value
    In 1=UE.Reroute(Value=UE.Multiply:Audio by Float(PrimaryOperand=@AudioMixerAudioMixerMono2_1.Out) {also {defaults {AdditionalOperands=0.43}}}.Out).Value
   ) {also {defaults {Gain 0=1.3}}}
   DecibelstoLinearGain_0 = UE.Decibels to Linear Gain:Float() {also {defaults {Decibels=-4}}}
   DecibelstoLinearGain_1 = UE.Decibels to Linear Gain:Float() {also {defaults {Decibels=-4}}}
   ADSREnvelopeADSREnvelope_2 = ADSR Envelope.ADSR Envelope:Float(Trigger Attack=@n1.Value) {also {defaults {Decay Time=1, Sustain Level=0}}}
   n42 = VariableAccessor.Float(Variable=VariableMutator.Float(
     Value=UE.Add:Float(PrimaryOperand=@ADSREnvelopeADSREnvelope_2.Out Envelope) {also {defaults {AdditionalOperands=1}}}.Out
     Variable=InitVariable.Float().Variable
    ).Variable)
   n43 = VariableAccessor.Float(Variable=@n42.Variable)
   n44 = UE.Add:Float(PrimaryOperand=VariableAccessor.Float(Variable=@n29.Variable).Value) {also {defaults {AdditionalOperands=1}}}
   n45 = UE.InterpTo:Audio(Target=/FortMetasound/Patches/ConditionalFloat(Condition=@n16.Value) {also {defaults {False=1, True=1.6788}}}.Output)
   n46 = &ConditionalFloat(Condition=@n17.Value) {also {defaults {False=0.7499, True=0.5957}}}
   n47 = UE.Reroute(Value=Max.Max:Float(B=^Fort.Vehicle.Throttle, A=^REV.Throttle).Value)
   n48 = VariableMutator.Float(Variable=InitVariable.Float().Variable, Value=@n47.Value)
   n49 = MapRange.MapRange:Float(In=^Fort.Vehicle.BonusSpeed) {also {defaults {Out Range B=0.75}}}
   REVRevPlayer_0 = REV.Rev Player:Stereo(
    REV Asset=^REV Asset
    Play=^UE.Source.OnPlay
    Pitch Shift=UE.Add:Float(AdditionalOperands=^Fort.Vehicle.Pitch, PrimaryOperand=UE.InterpTo:Audio(Target=@n49.Out Value) {also {defaults {Interp Time=0.15}}}.Value).Out
    Throttle=@n47.Value
    RPM=@n37.Value
    Gear=@n4.Value
   ) {also {defaults {REV Settings=/VehicleAudio/EngineAudio/Ferrari458/REVSetting_Default}}}
   BiquadFilter_0 = UE.Biquad Filter:Audio(
    Gain=MapRange.MapRange:Float(In=VariableAccessor.Float(Variable=@n48.Variable).Value) {also {defaults {Out Range A=3, Out Range B=-1}}}.Out Value
    In=@REVRevPlayer_0.Out Left
   ) {also {defaults {Cutoff Frequency=3500, Type=4}}}
   AudioMixerAudioMixerMono3_2 = "AudioMixer.Audio Mixer (Mono, 3)"(
    In 0=@AudioMixerAudioMixerMono3_0.Out
    In 2=@AudioMixerAudioMixerMono3_1.Out
    Gain 0=&MSP_AffiliationMixFloat(Affiliation=^Character.Affiliation, Enemy=@DecibelstoLinearGain_0.Linear Gain, Non-Enemy=@DecibelstoLinearGain_1.Linear Gain).Out
    Gain 2=UE.Multiply:Float(
     AdditionalOperands=MapRange.MapRange:Float(In=@n43.Value) {also {defaults {In Range A=1, In Range B=2, Out Range A=1, Out Range B=0.5}}}.Out Value
     PrimaryOperand=UE.Multiply:Float(AdditionalOperands=@n44.Out, PrimaryOperand=@n45.Value).Out
    ).Out
    Gain 1=UE.Multiply:Float(
     PrimaryOperand=UE.InterpTo:Audio(Target=@n46.Output).Value
     AdditionalOperands=MapRange.MapRange:Float(In=@n42.Value) {also {defaults {In Range A=1, In Range B=2, Out Range A=1, Out Range B=1.7}}}.Out Value
    ).Out
    In 1=UE.Reroute(Value=@BiquadFilter_0.Out).Value
   )
   BiquadFilter_1 = UE.Biquad Filter:Audio(In=@AudioMixerAudioMixerMono3_2.Out) {also {defaults {Cutoff Frequency=90, Gain=-6, Type=1}}}
   VariableMutator.Float(Value=^Fort.Vehicle.BonusSpeed, Variable=InitVariable.Float().Variable)
   VariableMutator.Float(Variable=InitVariable.Float().Variable, Value=@REVRevPlayer_0.Fundamental Frequency)
  ]
  links [{dst="@UE.OutputFormat.Mono.Audio:0", src=BiquadFilter_1.Out}]
 }]
properties {
 assetImportData=&self:AssetImportData
 attenuationSettings=&Vehicle_Engine_Att
 audioPropertiesSheet=&APS_Cosmetic_Vehicle_Engine
 bLooping=true
 duration=10000
 maxDistance=12000
 priority=80
 sampleRate=48000
 soundClassObject=&Fort_Vehicles_Engine
 soundSubmixObject=&Fortnite_SFX
 volume=1.4159
 blockRateOverride {default=28}
 concurrencySet [&Conc_Veh_EngineAudio_LowSpec, &Conc_Veh_EngineAudio_LowSpec_Actor]
 modulationSettings {
  highpassRouting=Union
  lowpassRouting=Union
  pitchRouting=Union
  volumeRouting=Union
  highpassModulationDestination {bEnableModulation=true, modulators [&PP_FortVehicles_Engine_HPF]}
  lowpassModulationDestination {bEnableModulation=true, modulators [&Patch_FortVehicles_Engine_LPF]}
  pitchModulationDestination {bEnableModulation=true, modulators [&DM_Bus_Vehicle_Engine_Pitch]}
  volumeModulationDestination {
   bEnableModulation=true
   modulators [&DM_VehicleEnginePatch, &Patch_FortVehicles_Engine_Vol]
  }
 }
}
